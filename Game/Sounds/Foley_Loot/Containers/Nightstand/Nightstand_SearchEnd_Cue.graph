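# Sound cue played when searching a nightstand container ends, in the loot foley set.
# A mixer layers the shared "no more loot" UI sound, at slightly reduced volume, with a randomly chosen one of two nightstand search-end waves.
# A modulator varies pitch and volume of the nightstand foley only.
# The cue uses prop-search attenuation, the loot-collect sound class and a concurrency override that stops the farthest instance first.
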SoundCue /Game/Sounds/Foley_Loot/Containers/Nightstand/Nightstand_SearchEnd_Cue
root=Mixer_6
nodes [
 {id=Mixer_6, children [WavePlayer_31, Modulator_5], properties {inputVolume [0.8,1]}}
 {id=Modulator_5, children [Random_1], properties {pitchMax=1.07, pitchMin=0.92, volumeMax=1.07, volumeMin=0.92}}
 {id=Random_1, children [WavePlayer_5,WavePlayer_6], properties {weights [1,1]}}
 {id=WavePlayer_31, properties {soundWaveAssetPtr=/Game/Sounds/Fort_UI_Sounds/Gameplay/Fort_UI_NoMoreLoot}}
 {id=WavePlayer_5, properties {soundWaveAssetPtr=&/Nightstand_SearchEnd1}}
 {id=WavePlayer_6, properties {soundWaveAssetPtr=&/Nightstand_SearchEnd2}}
]
properties {
 attenuationSettings=/Game/Sounds/Attenuation/Fort_Prop_Search_Att
 bOverrideConcurrency=true
 maxDistance=1300
 soundClassObject=&Fort_Loot_Collect
 volumeMultiplier=0.4
 concurrencyOverrides {maxCount=4, resolutionRule=StopFarthestThenPreventNew}
}
