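# Sound cue for searching a nightstand loot container, routed to the loot-collect sound class with prop-search attenuation.
# One of three search waves is picked at random, then modulated with slightly varied volume and raised pitch.
# A mixer plays that shared chain twice, immediately and again quieter after a fixed one-second delay, to layer the rummage.
# A concurrency override prevents new instances once the limit is reached.
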SoundCue /Game/Sounds/Foley_Loot/Containers/Nightstand/Nightstand_Search_Cue
root=Mixer_37
nodes [
 {id=Delay_58, children [Modulator_2], properties {delayMax=1, delayMin=1}}
 {id=Mixer_37, children [Modulator_2,Delay_58], properties {inputVolume [0.8,0.6]}}
 {id=Modulator_2, children [Random_0], properties {pitchMax=1.3, pitchMin=1.1, volumeMax=1.07, volumeMin=0.92}}
 {id=Random_0, children [WavePlayer_0,WavePlayer_1,WavePlayer_3], properties {weights [1,1,1]}}
 {id=WavePlayer_0, properties {soundWaveAssetPtr=&/Nightstand_Search02}}
 {id=WavePlayer_1, properties {soundWaveAssetPtr=&/Nightstand_Search04}}
 {id=WavePlayer_3, properties {soundWaveAssetPtr=&/Nightstand_Search08}}
]
properties {
 attenuationSettings=/Game/Sounds/Attenuation/Fort_Prop_Search_Att
 bOverrideConcurrency=true
 maxDistance=1300
 soundClassObject=&Fort_Loot_Collect
 volumeMultiplier=0.45
 concurrencyOverrides {resolutionRule=PreventNew}
}
